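; BP_S30_Story_05_GemstoneCache is a quest actor for Season 30 Story 05 that reports when a player damages the hidden gemstone cache.
; On BeginPlay it calls the parent and binds a handler to its OnDamaged dispatcher.
; The handler casts the instigator to a Fort player controller, builds a HiddenCache damage custom verb message with itself tagged as DamagedProp, and broadcasts it.
; Overlap, Tick and the construction script add no logic of their own.
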(event EventBeginPlay
  (|CallParentFunction)
  (EventDispatchers|BindEventtoOnDamaged (delegate Custom|OnDamaged_Event)))

(event Collision|EventActorBeginOverlap
  (params (OtherActor :type Actor%20Object%20Reference)))

(event Custom|OnDamaged_Event
  (params (DamagedActor :type Actor%20Object%20Reference) (Damage :type Float%20%28single-precision%29) (InstigatedBy :type Controller%20Object%20Reference) (DamageCauser :type Actor%20Object%20Reference) (HitLocation :type Vector) (FHitComponent :type Primitive%20Component%20Object%20Reference) (BoneName :type Name) (Momentum :type Vector))
  (bind _asfort_player_controller (Utilities|Casting|CastToFortPlayerController InstigatedBy))
  (bind _result (CustomVerb|AddTagsfromSubject (CustomVerb|MakeCustomVerbMessage _asfort_player_controller "(TagName=\"CustomVerb.Damage.S30.HiddenCache\")") "DamagedProp" self))
  (CustomVerb|BroadcastCustomVerbMessage _result))

(event EventTick
  (params (DeltaSeconds :type Float%20%28single-precision%29)))

(fn ConstructionScript
  (|CallParentFunction))
