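; Story quest gemstone VFX actor for BattlePass S30 Story 05 that cycles through constellation shapes with sound.
; Construction gives the pre-interaction mesh a dynamic material and sets its flipbook animation phase from a per-gem value.
; The pre-interaction VFX event starts cosmetic audio and a looping timer at the switch frequency.
; Each tick of that timer advances a shape index modulo four, passes it to the ambient Niagara component and replays the audio.
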(event EventBeginPlay
  (|CallParentFunction))

(event Collision|EventActorBeginOverlap
  (params (OtherActor :type Actor%20Object%20Reference)))

(event EventTick
  (params (DeltaSeconds :type Float%20%28single-precision%29)))

(event VisualState|VFX&Audio|EventPreInteractionVFX&Audio
  (params (PlayerController :type Fort%20Player%20Controller%20Object%20Reference))
  (|CallParentFunction PlayerController)
  (Utilities|Time|SetTimerbyEvent (delegate Custom|FX_SwitchConstelation) (Variables|FX|GetSwitchFrequency) true true)
  (Audio|Components|Audio|Play (Variables|Components|GetCosmeticAudio) 0.0))

(event Custom|FX_SwitchConstelation
  (bind _output_get (Variables|FX|SetShapeIndex (Math|Integer|%25%28Integer%29 (+ (Variables|FX|GetShapeIndex) 1) 4)))
  (Niagara|SetNiagaraVariable(Int32) (Variables|Components|GetN_Ambient) "NextShapeIndex" _output_get)
  (Audio|Components|Audio|Play (Variables|Components|GetCosmeticAudio) (Variables|FX|GetSwitchFrequency)))

(fn ConstructionScript
  (|CallParentFunction)
  (bind _returnvalue (Rendering|Material|CreateDynamicMaterialInstance (Rendering|Material|GetMaterial (Variables|Components|GetSM_Quest_PreInteraction) 0) "None" "None" :!class /Script/Engine.KismetMaterialLibrary))
  (Variables|Default|SetGemMI _returnvalue)
  (Rendering|Material|SetMaterial (Variables|Components|GetSM_Quest_PreInteraction) 0 (Variables|Default|GetGemMI))
  (Audio|Parameter|SetBooleanParameter(Message) (Variables|Default|GetGemMI) "StaticFlipbookAnimPhase" true)
  (Rendering|Material|SetScalarParameterValue (Variables|Default|GetGemMI) "FlipbookAnimPhase" (Variables|Default|GetGemFlipbookValue) :!class /Script/Engine.MaterialInstanceDynamic))
